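# Material function in the SonnetSpirit drum kit cosmetic folder that converts a flipbook frame Index into a UV offset, Position, to add to texture coordinates.
# It takes column and row counts, both previewing as 4, and shares them through named reroutes.
# Fmod, Divide and Floor derive the wrapped column and row cell, which are normalized by the grid size and appended.
# It only outputs the offset, with no UV scaling or texture sampling.
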MaterialFunction /SparksCosmetics/Instrument/DrumKit/SonnetSpirit/Materials/MF_Flipbook_Index
graph {nodes [
  n1 = FunctionInput(Preview=Constant()) {bUsePreviewValueAsDefault=true, inputName=Index, inputType=FunctionInput_Scalar, sortPriority=4}
  n2 = NamedRerouteUsage()
  n3 = Floor(None=Fmod(A=@n1, B=@n2))
  n4 = Floor(None=@n2)
  n5 = NamedRerouteUsage()
  n6 = NamedRerouteUsage()
  n7 = Multiply(A=@n5, B=@n6)
  n8 = Floor(None=Divide(A=Fmod(A=@n1, B=@n7), B=@n5))
  n9 = Floor(None=@n6)
  FunctionOutput(None=AppendVector(A=Divide(A=@n3, B=@n4), B=Divide(A=@n8, B=@n9))) {description=Add to texture coordinates, outputName=Position, sortPriority=3}
  NamedRerouteDeclaration(None=FunctionInput(Preview=Constant() {r=4}) {bUsePreviewValueAsDefault=true, inputName=Number of Columns, inputType=FunctionInput_Scalar, sortPriority=3}) {name=Columns, nodeColor=<rgba 0.1719 0 1 1>}
  NamedRerouteDeclaration(None=FunctionInput(Preview=Constant() {r=4}) {bUsePreviewValueAsDefault=true, inputName=Number of Rows, inputType=FunctionInput_Scalar, sortPriority=2}) {name=Rows, nodeColor=<rgba 0.1719 0 1 1>}
 ]}
