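# Material function that draws animated flipbook faces on the Sonnet Spirit drum kit, blending FaceColor over the Lit and Shaded colors.
# Six parameter-placed faces (A–D eyes and mouth, E–F one half only) are masked and summed into the blend alpha.
# Each Hit time briefly overrides the eye frame for about 1/6 second.
# Expression indices select texture row and packed color channel, sampled with original-UV derivatives; a time-driven test face is included.
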MaterialFunction /SparksCosmetics/Instrument/DrumKit/SonnetSpirit/Materials/MF_SonnetSpirit_DrumKit
graph {nodes [
  Composite() {subgraphName=Transform tests}
  n1 = Constant() {r=0.2}
  n2 = VectorParameter() {group=Flipbook, parameterName=FaceColor, defaultValue=<rgba 0 0 0 1>}
  n3 = NamedRerouteUsage()
  n4 = NamedRerouteUsage()
  n5 = Add(A=NamedRerouteUsage(), B=NamedRerouteUsage())
  n6 = Add(A=NamedRerouteUsage(), B=NamedRerouteUsage())
  n7 = Saturate(None=Add(A=Add(A=@n3, B=@n4), B=Add(A=@n5, B=@n6)))
  FunctionOutput(None=LinearInterpolate(A=FunctionInput(Preview=@n1) {bUsePreviewValueAsDefault=true, inputName=ColorLit}, B=@n2.RGB, Alpha=@n7) {constB=0}) {outputName=Lit}
  FunctionOutput(None=LinearInterpolate(A=FunctionInput(Preview=@n1) {bUsePreviewValueAsDefault=true, inputName=ColorShaded, sortPriority=1}, B=@n2.RGB, Alpha=@n7) {constB=0}) {outputName=Shaded, sortPriority=1}
  n8 = Constant() {r=4}
  n9 = ScalarParameter() {group=Flipbook C, parameterName=C_EyesIndex, sortPriority=7}
  n10 = Subtract(A=Time() {period=2}, B=ScalarParameter() {defaultValue=1, group=Flipbook C, parameterName=C_Hit, sortPriority=9})
  n11 = Saturate(None=Floor(None=Multiply(A=@n10) {constB=6}))
  NamedRerouteDeclaration(None=AppendVector(
    A=LinearInterpolate(A=@n8, B=@n9, Alpha=@n11) {constA=4, constB=2}
    B=ScalarParameter() {defaultValue=2, group=Flipbook C, parameterName=C_MouthIndex, sortPriority=8}
   )) {name=Face C, nodeColor=<rgba 0.0547 1 0 1>}
  n12 = Constant() {r=4}
  n13 = ScalarParameter() {group=Flipbook D, parameterName=D_EyesIndex, sortPriority=7}
  n14 = Subtract(A=Time() {period=2}, B=ScalarParameter() {defaultValue=1, group=Flipbook D, parameterName=D_Hit, sortPriority=9})
  n15 = Saturate(None=Floor(None=Multiply(A=@n14) {constB=6}))
  NamedRerouteDeclaration(None=AppendVector(
    A=LinearInterpolate(A=@n12, B=@n13, Alpha=@n15) {constA=4, constB=2}
    B=ScalarParameter() {defaultValue=2, group=Flipbook D, parameterName=D_MouthIndex, sortPriority=8}
   )) {name=Face D, nodeColor=<rgba 0.0547 1 0 1>}
  n16 = Constant() {desc=Rows, r=2}
  n17 = Constant() {desc=Columns, r=1}
  n18 = AppendVector(A=Time() {period=2}, B=Divide(A=Time() {period=2}) {constB=1.23})
  n19 = Fmod(A=Divide(A=@n18) {constB=3}, B=Constant() {r=2})
  n20 = MaterialFunctionCall!(Number of Rows=@n16, Number of Columns=@n17, Index=ComponentMask(None=@n19) {r=true}) {materialFunction=&/MF_Flipbook_Index}
  n21 = MaterialFunctionCall(Number of Rows=@n16, Number of Columns=@n17, Index=ComponentMask(None=@n19) {g=true}) {materialFunction=&MF_Flipbook_Index}
  n22 = Time() {period=2}
  n23 = Sine(None=Multiply(A=@n22) {constB=1.25})
  n24 = Sine(None=@n22)
  n25 = MaterialFunctionCall!(
   UVs=TextureCoordinate() {coordinateIndex=1}
   RowsScaleCorrection=Constant() {r=0.5}
   OffsetScale=Divide(A=Add(A=@n23) {constB=5}) {constB=14}
   OffsetRotation=MaterialFunctionCall!(Value=Multiply(A=@n22) {constB=60}, Wrap=Constant() {r=360}) {materialFunction=/Engine/Functions/Engine_MaterialFunctions03/Math/WrapFloat}.Result
   OffsetU=Divide(A=@n24) {constB=7}
   OffsetV=Divide(A=@n24) {constB=9}
   Number of Columns=Constant() {desc=Columns, r=1}
   Number of Rows=Constant() {desc=Rows, r=2}
  ) {materialFunction=&/MF_Flipbook_Transform}
  n26 = Divide(B=Constant() {desc=Rows, r=2})
  n27 = Round(None=Add(A=ComponentMask(None=@n25.Coordinates) {g=true}, B=Divide(A=@n26)) {constB=0.1})
  n28 = TextureCoordinate() {coordinateIndex=1}
  n29 = TextureSampleParameter2D!(
   UVs=Add(A=@n25.Coordinates, B=LinearInterpolate(A=@n20.Position, B=@n21.Position, Alpha=@n27) {constA=3, constAlpha=0})
   "DDX(UVs)"=DDX(Value=@n28)
   "DDY(UVs)"=DDY(Value=@n28)
   Apply View MipBias
  ) {
   mipValueMode=TMVM_Derivative
   parameterName=FaceExpressions
   samplerType=SAMPLERTYPE_LinearColor
   texture=/SparksCosmetics/Instrument/DrumKit/SonnetSpirit/Textures/T_SonnetSpirit_DrumKit_Expressions_FB
  }
  n30 = Floor(None=Fmod(A=@n18, B=Constant() {r=3}))
  n31 = Saturate(None=@n30)
  n32 = ComponentMask(None=@n31) {r=true}
  n33 = ComponentMask(None=@n31) {g=true}
  n34 = Saturate(None=Subtract(A=@n30))
  n35 = LinearInterpolate(
   A=LinearInterpolate(A=@n29.R, B=@n29.R, Alpha=LinearInterpolate(A=@n32, B=@n33, Alpha=@n27) {constA=3, constAlpha=0}) {constAlpha=1, constB=0}
   B=@n29.B
   Alpha=LinearInterpolate(A=ComponentMask(None=@n34) {r=true}, B=ComponentMask(None=@n34) {g=true}, Alpha=@n27) {constA=3, constAlpha=0}
  ) {constAlpha=1, constB=0}
  NamedRerouteDeclaration(None=LinearInterpolate(Alpha=Multiply(A=@n35, B=@n25.Mask)) {constA=0.05, constB=0}) {name=A_Face_6, nodeColor=<rgba 0 0.7188 1 1>}
  n36 = Time() {period=2}
  n37 = ScalarParameter() {defaultValue=1, group=Flipbook F, parameterName=F_Hit, sortPriority=9}
  n38 = Floor(None=Multiply(A=Subtract(A=@n36, B=@n37)) {constB=6})
  NamedRerouteDeclaration(None=LinearInterpolate(
    A=Constant() {r=4}
    B=ScalarParameter() {defaultValue=4, group=Flipbook F, parameterName=F_EyesIndex, sortPriority=7}
    Alpha=Saturate(None=@n38)
   ) {constA=4, constB=2}) {name=Face F, nodeColor=<rgba 0.0547 1 0 1>}
  n39 = MaterialFunctionCall(
   UVs=TextureCoordinate() {coordinateIndex=1}
   RowsScaleCorrection=Constant() {r=0.5}
   OffsetScale=ScalarParameter() {defaultValue=0.3, group=Flipbook A, parameterName=A_OffsetScale, sortPriority=1}
   OffsetRotation=ScalarParameter() {group=Flipbook A, parameterName=A_OffsetRotation, sortPriority=2}
   OffsetU=ScalarParameter() {defaultValue=0.2, group=Flipbook A, parameterName=A_OffsetU, sortPriority=3}
   OffsetV=ScalarParameter() {group=Flipbook A, parameterName=A_OffsetV, sortPriority=4}
   Number of Columns=Constant() {desc=Columns, r=1}
   Number of Rows=Constant() {desc=Rows, r=2}
  ) {materialFunction=&MF_Flipbook_Transform}
  n40 = Constant() {desc=Rows, r=2}
  n41 = Constant() {desc=Columns, r=1}
  n42 = Fmod(A=Divide(A=NamedRerouteUsage()) {constB=3}, B=Constant() {r=2})
  n43 = MaterialFunctionCall(Number of Rows=@n40, Number of Columns=@n41, Index=ComponentMask(None=@n42) {r=true}) {materialFunction=&MF_Flipbook_Index}
  n44 = MaterialFunctionCall(Number of Rows=@n40, Number of Columns=@n41, Index=ComponentMask(None=@n42) {g=true}) {materialFunction=&MF_Flipbook_Index}
  n45 = Divide(B=Constant() {desc=Rows, r=2})
  n46 = Round(None=Add(A=ComponentMask(None=@n39.Coordinates) {g=true}, B=Divide(A=@n45)) {constB=0.1})
  n47 = TextureCoordinate() {coordinateIndex=1}
  n48 = TextureSampleParameter2D(
   UVs=Add(A=@n39.Coordinates, B=LinearInterpolate(A=@n43.Position, B=@n44.Position, Alpha=@n46) {constA=3, constAlpha=0})
   "DDX(UVs)"=DDX(Value=@n47)
   "DDY(UVs)"=DDY(Value=@n47)
  ) {
   group=Flipbook
   mipValueMode=TMVM_Derivative
   parameterName=FaceExpressions
   samplerType=SAMPLERTYPE_LinearColor
   texture=&T_SonnetSpirit_DrumKit_Expressions_FB
  }
  n49 = Floor(None=Fmod(A=NamedRerouteUsage(), B=Constant() {r=3}))
  n50 = Saturate(None=@n49)
  n51 = LinearInterpolate(
   A=@n48.R
   B=@n48.R
   Alpha=LinearInterpolate(A=ComponentMask(None=@n50) {r=true}, B=ComponentMask(None=@n50) {g=true}, Alpha=@n46) {constA=3, constAlpha=0}
  ) {constAlpha=1, constB=0}
  n52 = Saturate(None=Subtract(A=@n49))
  n53 = LinearInterpolate(A=ComponentMask(None=@n52) {r=true}, B=ComponentMask(None=@n52) {g=true}, Alpha=@n46) {constA=3, constAlpha=0}
  NamedRerouteDeclaration(None=Multiply(A=LinearInterpolate(A=@n51, B=@n48.B, Alpha=@n53) {constAlpha=1, constB=0}, B=@n39.Mask)) {name=A_Face, nodeColor=<rgba 0 0.7187 1 1>}
  n54 = MaterialFunctionCall(
   UVs=TextureCoordinate() {coordinateIndex=1}
   RowsScaleCorrection=Constant() {r=0.5}
   OffsetScale=ScalarParameter() {defaultValue=0.35, group=Flipbook B, parameterName=B_OffsetScale, sortPriority=1}
   OffsetRotation=ScalarParameter() {group=Flipbook B, parameterName=B_OffsetRotation, sortPriority=2}
   OffsetU=ScalarParameter() {defaultValue=-0.25, group=Flipbook B, parameterName=B_OffsetU, sortPriority=3}
   OffsetV=ScalarParameter() {defaultValue=-0.25, group=Flipbook B, parameterName=B_OffsetV, sortPriority=4}
   Number of Columns=Constant() {desc=Columns, r=1}
   Number of Rows=Constant() {desc=Rows, r=2}
  ) {materialFunction=&MF_Flipbook_Transform}
  n55 = Constant() {desc=Rows, r=2}
  n56 = Constant() {desc=Columns, r=1}
  n57 = Fmod(A=Divide(A=NamedRerouteUsage()) {constB=3}, B=Constant() {r=2})
  n58 = MaterialFunctionCall(Number of Rows=@n55, Number of Columns=@n56, Index=ComponentMask(None=@n57) {r=true}) {materialFunction=&MF_Flipbook_Index}
  n59 = MaterialFunctionCall(Number of Rows=@n55, Number of Columns=@n56, Index=ComponentMask(None=@n57) {g=true}) {materialFunction=&MF_Flipbook_Index}
  n60 = Divide(B=Constant() {desc=Rows, r=2})
  n61 = Round(None=Add(A=ComponentMask(None=@n54.Coordinates) {g=true}, B=Divide(A=@n60)) {constB=0.1})
  n62 = TextureCoordinate() {coordinateIndex=1}
  n63 = TextureSampleParameter2D(
   UVs=Add(A=@n54.Coordinates, B=LinearInterpolate(A=@n58.Position, B=@n59.Position, Alpha=@n61) {constA=3, constAlpha=0})
   "DDX(UVs)"=DDX(Value=@n62)
   "DDY(UVs)"=DDY(Value=@n62)
  ) {
   mipValueMode=TMVM_Derivative
   parameterName=FaceExpressions
   samplerType=SAMPLERTYPE_LinearColor
   texture=&T_SonnetSpirit_DrumKit_Expressions_FB
  }
  n64 = Floor(None=Fmod(A=NamedRerouteUsage(), B=Constant() {r=3}))
  n65 = Saturate(None=@n64)
  n66 = LinearInterpolate(
   A=@n63.R
   B=@n63.R
   Alpha=LinearInterpolate(A=ComponentMask(None=@n65) {r=true}, B=ComponentMask(None=@n65) {g=true}, Alpha=@n61) {constA=3, constAlpha=0}
  ) {constAlpha=1, constB=0}
  n67 = Saturate(None=Subtract(A=@n64))
  n68 = LinearInterpolate(A=ComponentMask(None=@n67) {r=true}, B=ComponentMask(None=@n67) {g=true}, Alpha=@n61) {constA=3, constAlpha=0}
  NamedRerouteDeclaration(None=Multiply(A=LinearInterpolate(A=@n66, B=@n63.B, Alpha=@n68) {constAlpha=1, constB=0}, B=@n54.Mask)) {name=A_Face_1, nodeColor=<rgba 0 0.7188 1 1>}
  n69 = MaterialFunctionCall(
   UVs=TextureCoordinate() {coordinateIndex=1}
   RowsScaleCorrection=Constant() {r=0.5}
   OffsetScale=ScalarParameter() {defaultValue=0.35, group=Flipbook C, parameterName=C_OffsetScale, sortPriority=1}
   OffsetRotation=ScalarParameter() {group=Flipbook C, parameterName=C_OffsetRotation, sortPriority=2}
   OffsetU=ScalarParameter() {group=Flipbook C, parameterName=C_OffsetU, sortPriority=3}
   OffsetV=ScalarParameter() {defaultValue=0.5, group=Flipbook C, parameterName=C_OffsetV, sortPriority=4}
   Number of Columns=Constant() {desc=Columns, r=1}
   Number of Rows=Constant() {desc=Rows, r=2}
  ) {materialFunction=&MF_Flipbook_Transform}
  n70 = Constant() {desc=Rows, r=2}
  n71 = Constant() {desc=Columns, r=1}
  n72 = Fmod(A=Divide(A=NamedRerouteUsage()) {constB=3}, B=Constant() {r=2})
  n73 = MaterialFunctionCall(Number of Rows=@n70, Number of Columns=@n71, Index=ComponentMask(None=@n72) {r=true}) {materialFunction=&MF_Flipbook_Index}
  n74 = MaterialFunctionCall(Number of Rows=@n70, Number of Columns=@n71, Index=ComponentMask(None=@n72) {g=true}) {materialFunction=&MF_Flipbook_Index}
  n75 = Divide(B=Constant() {desc=Rows, r=2})
  n76 = Round(None=Add(A=ComponentMask(None=@n69.Coordinates) {g=true}, B=Divide(A=@n75)) {constB=0.1})
  n77 = TextureCoordinate() {coordinateIndex=1}
  n78 = TextureSampleParameter2D(
   UVs=Add(A=@n69.Coordinates, B=LinearInterpolate(A=@n73.Position, B=@n74.Position, Alpha=@n76) {constA=3, constAlpha=0})
   "DDX(UVs)"=DDX(Value=@n77)
   "DDY(UVs)"=DDY(Value=@n77)
  ) {
   mipValueMode=TMVM_Derivative
   parameterName=FaceExpressions
   samplerType=SAMPLERTYPE_LinearColor
   texture=&T_SonnetSpirit_DrumKit_Expressions_FB
  }
  n79 = Floor(None=Fmod(A=NamedRerouteUsage(), B=Constant() {r=3}))
  n80 = Saturate(None=@n79)
  n81 = LinearInterpolate(
   A=@n78.R
   B=@n78.R
   Alpha=LinearInterpolate(A=ComponentMask(None=@n80) {r=true}, B=ComponentMask(None=@n80) {g=true}, Alpha=@n76) {constA=3, constAlpha=0}
  ) {constAlpha=1, constB=0}
  n82 = Saturate(None=Subtract(A=@n79))
  n83 = LinearInterpolate(A=ComponentMask(None=@n82) {r=true}, B=ComponentMask(None=@n82) {g=true}, Alpha=@n76) {constA=3, constAlpha=0}
  NamedRerouteDeclaration(None=Multiply(A=LinearInterpolate(A=@n81, B=@n78.B, Alpha=@n83) {constAlpha=1, constB=0}, B=@n69.Mask)) {name=A_Face_2, nodeColor=<rgba 0 0.7188 1 1>}
  n84 = MaterialFunctionCall(
   UVs=TextureCoordinate() {coordinateIndex=1}
   RowsScaleCorrection=Constant() {r=0.5}
   OffsetScale=ScalarParameter() {defaultValue=0.35, group=Flipbook D, parameterName=D_OffsetScale, sortPriority=1}
   OffsetRotation=ScalarParameter() {group=Flipbook D, parameterName=D_OffsetRotation, sortPriority=2}
   OffsetU=ScalarParameter() {group=Flipbook D, parameterName=D_OffsetU, sortPriority=3}
   OffsetV=ScalarParameter() {group=Flipbook D, parameterName=D_OffsetV, sortPriority=4}
   Number of Columns=Constant() {desc=Columns, r=1}
   Number of Rows=Constant() {desc=Rows, r=2}
  ) {materialFunction=&MF_Flipbook_Transform}
  n85 = Constant() {desc=Rows, r=2}
  n86 = Constant() {desc=Columns, r=1}
  n87 = Fmod(A=Divide(A=NamedRerouteUsage()) {constB=3}, B=Constant() {r=2})
  n88 = MaterialFunctionCall(Number of Rows=@n85, Number of Columns=@n86, Index=ComponentMask(None=@n87) {r=true}) {materialFunction=&MF_Flipbook_Index}
  n89 = MaterialFunctionCall(Number of Rows=@n85, Number of Columns=@n86, Index=ComponentMask(None=@n87) {g=true}) {materialFunction=&MF_Flipbook_Index}
  n90 = Divide(B=Constant() {desc=Rows, r=2})
  n91 = Round(None=Add(A=ComponentMask(None=@n84.Coordinates) {g=true}, B=Divide(A=@n90)) {constB=0.1})
  n92 = TextureCoordinate() {coordinateIndex=1}
  n93 = TextureSampleParameter2D(
   UVs=Add(A=@n84.Coordinates, B=LinearInterpolate(A=@n88.Position, B=@n89.Position, Alpha=@n91) {constA=3, constAlpha=0})
   "DDX(UVs)"=DDX(Value=@n92)
   "DDY(UVs)"=DDY(Value=@n92)
  ) {
   mipValueMode=TMVM_Derivative
   parameterName=FaceExpressions
   samplerType=SAMPLERTYPE_LinearColor
   texture=&T_SonnetSpirit_DrumKit_Expressions_FB
  }
  n94 = Floor(None=Fmod(A=NamedRerouteUsage(), B=Constant() {r=3}))
  n95 = Saturate(None=@n94)
  n96 = LinearInterpolate(
   A=@n93.R
   B=@n93.R
   Alpha=LinearInterpolate(A=ComponentMask(None=@n95) {r=true}, B=ComponentMask(None=@n95) {g=true}, Alpha=@n91) {constA=3, constAlpha=0}
  ) {constAlpha=1, constB=0}
  n97 = Saturate(None=Subtract(A=@n94))
  n98 = LinearInterpolate(A=ComponentMask(None=@n97) {r=true}, B=ComponentMask(None=@n97) {g=true}, Alpha=@n91) {constA=3, constAlpha=0}
  NamedRerouteDeclaration(None=Multiply(A=LinearInterpolate(A=@n96, B=@n93.B, Alpha=@n98) {constAlpha=1, constB=0}, B=@n84.Mask)) {name=A_Face_3, nodeColor=<rgba 0 0.7188 1 1>}
  n99 = Constant() {r=4}
  n100 = ScalarParameter() {group=Flipbook A, parameterName=A_EyesIndex, sortPriority=7}
  n101 = Subtract(A=Time() {period=2}, B=ScalarParameter() {defaultValue=1, group=Flipbook A, parameterName=A_Hit, sortPriority=9})
  n102 = Saturate(None=Floor(None=Multiply(A=@n101) {constB=6}))
  NamedRerouteDeclaration(None=AppendVector(
    A=LinearInterpolate(A=@n99, B=@n100, Alpha=@n102) {constA=4, constB=2}
    B=ScalarParameter() {defaultValue=6, group=Flipbook A, parameterName=A_MouthIndex, sortPriority=8}
   )) {name=Face A, nodeColor=<rgba 0.0547 1 0 1>}
  n103 = MaterialFunctionCall(
   UVs=TextureCoordinate() {coordinateIndex=1}
   RowsScaleCorrection=Constant() {r=0.5}
   OffsetScale=ScalarParameter() {defaultValue=0.35, group=Flipbook E, parameterName=E_OffsetScale, sortPriority=1}
   OffsetRotation=ScalarParameter() {group=Flipbook E, parameterName=E_OffsetRotation, sortPriority=2}
   OffsetU=ScalarParameter() {defaultValue=0.3, group=Flipbook E, parameterName=E_OffsetU, sortPriority=3}
   OffsetV=ScalarParameter() {defaultValue=0.5, group=Flipbook E, parameterName=E_OffsetV, sortPriority=4}
   Number of Columns=Constant() {desc=Columns, r=1}
   Number of Rows=Constant() {desc=Rows, r=2}
  ) {materialFunction=&MF_Flipbook_Transform}
  n104 = Constant() {desc=Rows, r=2}
  n105 = Constant() {desc=Columns, r=1}
  n106 = NamedRerouteUsage()
  n107 = ComponentMask(None=Fmod(A=Divide(A=@n106) {constB=3}, B=Constant() {r=2})) {r=true}
  n108 = TextureCoordinate() {coordinateIndex=1}
  n109 = TextureSampleParameter2D(
   UVs=Add(
    A=@n103.Coordinates
    B=MaterialFunctionCall(Number of Rows=@n104, Number of Columns=@n105, Index=@n107) {materialFunction=&MF_Flipbook_Index}.Position
   )
   "DDX(UVs)"=DDX(Value=@n108)
   "DDY(UVs)"=DDY(Value=@n108)
  ) {
   mipValueMode=TMVM_Derivative
   parameterName=FaceExpressions
   samplerType=SAMPLERTYPE_LinearColor
   texture=&T_SonnetSpirit_DrumKit_Expressions_FB
  }
  n110 = Floor(None=Fmod(A=NamedRerouteUsage(), B=Constant() {r=3}))
  n111 = LinearInterpolate(A=@n109.R, B=@n109.R, Alpha=Saturate(None=@n110)) {constAlpha=1, constB=0}
  n112 = Saturate(None=Subtract(A=@n110))
  n113 = ComponentMask(None=@n103.Coordinates) {g=true}
  n114 = Divide(A=Divide(B=Constant() {desc=Rows, r=2}))
  n115 = OneMinus(None=Round(None=Add(A=@n113, B=@n114) {constB=0.1}))
  NamedRerouteDeclaration(None=Multiply(A=LinearInterpolate(A=@n111, B=@n109.B, Alpha=@n112) {constAlpha=1, constB=0}, B=Multiply(A=@n103.Mask, B=@n115))) {name=A_Face_4, nodeColor=<rgba 0 0.7188 1 1>}
  n116 = MaterialFunctionCall(
   UVs=TextureCoordinate() {coordinateIndex=1}
   RowsScaleCorrection=Constant() {r=0.5}
   OffsetScale=ScalarParameter() {defaultValue=0.35, group=Flipbook F, parameterName=F_OffsetScale, sortPriority=2}
   OffsetRotation=ScalarParameter() {group=Flipbook F, parameterName=F_OffsetRotation, sortPriority=2}
   OffsetU=ScalarParameter() {defaultValue=-0.3, group=Flipbook F, parameterName=F_OffsetU, sortPriority=3}
   OffsetV=ScalarParameter() {defaultValue=0.25, group=Flipbook F, parameterName=F_OffsetV, sortPriority=4}
   Number of Columns=Constant() {desc=Columns, r=1}
   Number of Rows=Constant() {desc=Rows, r=2}
  ) {materialFunction=&MF_Flipbook_Transform}
  n117 = Constant() {desc=Rows, r=2}
  n118 = Constant() {desc=Columns, r=1}
  n119 = NamedRerouteUsage()
  n120 = ComponentMask(None=Fmod(A=Divide(A=@n119) {constB=3}, B=Constant() {r=2})) {r=true}
  n121 = TextureCoordinate() {coordinateIndex=1}
  n122 = TextureSampleParameter2D(
   UVs=Add(
    A=@n116.Coordinates
    B=MaterialFunctionCall(Number of Rows=@n117, Number of Columns=@n118, Index=@n120) {materialFunction=&MF_Flipbook_Index}.Position
   )
   "DDX(UVs)"=DDX(Value=@n121)
   "DDY(UVs)"=DDY(Value=@n121)
  ) {
   mipValueMode=TMVM_Derivative
   parameterName=FaceExpressions
   samplerType=SAMPLERTYPE_LinearColor
   texture=&T_SonnetSpirit_DrumKit_Expressions_FB
  }
  n123 = Floor(None=Fmod(A=NamedRerouteUsage(), B=Constant() {r=3}))
  n124 = LinearInterpolate(A=@n122.R, B=@n122.R, Alpha=Saturate(None=@n123)) {constAlpha=1, constB=0}
  n125 = Saturate(None=Subtract(A=@n123))
  n126 = ComponentMask(None=@n116.Coordinates) {g=true}
  n127 = Divide(A=Divide(B=Constant() {desc=Rows, r=2}))
  n128 = OneMinus(None=Round(None=Add(A=@n126, B=@n127) {constB=0.1}))
  NamedRerouteDeclaration(None=Multiply(A=LinearInterpolate(A=@n124, B=@n122.B, Alpha=@n125) {constAlpha=1, constB=0}, B=Multiply(A=@n116.Mask, B=@n128))) {name=A_Face_5, nodeColor=<rgba 0 0.7188 1 1>}
  n129 = Constant() {r=4}
  n130 = ScalarParameter() {group=Flipbook B, parameterName=B_EyesIndex, sortPriority=7}
  n131 = Subtract(A=Time() {period=2}, B=ScalarParameter() {defaultValue=1, group=Flipbook B, parameterName=B_Hit, sortPriority=9})
  n132 = Saturate(None=Floor(None=Multiply(A=@n131) {constB=6}))
  NamedRerouteDeclaration(None=AppendVector(
    A=LinearInterpolate(A=@n129, B=@n130, Alpha=@n132) {constA=4, constB=2}
    B=ScalarParameter() {defaultValue=2, group=Flipbook B, parameterName=B_MouthIndex, sortPriority=8}
   )) {name=Face B, nodeColor=<rgba 0.0547 1 0 1>}
  n133 = Time() {period=2}
  n134 = ScalarParameter() {defaultValue=1, group=Flipbook E, parameterName=E_Hit, sortPriority=9}
  n135 = Floor(None=Multiply(A=Subtract(A=@n133, B=@n134)) {constB=6})
  NamedRerouteDeclaration(None=LinearInterpolate(A=Constant() {r=4}, B=ScalarParameter() {group=Flipbook E, parameterName=E_EyesIndex, sortPriority=7}, Alpha=Saturate(None=@n135)) {constA=4, constB=2}) {name=Face E, nodeColor=<rgba 0.0547 1 0 1>}
  PinBase()
  PinBase()
  Time() {period=2}
 ]}
